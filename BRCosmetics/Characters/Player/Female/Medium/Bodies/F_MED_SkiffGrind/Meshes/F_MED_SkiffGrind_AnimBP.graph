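# Post-copy animation Blueprint for the SkiffGrind female medium body cosmetic, adding secondary motion to its left and right loop bones.
# It copies the attached parent mesh's pose, runs the body's Control Rig, then weight-constrains both loop base bones.
# The left constraint blends upper-arm twist with its own loop bone, while the right uses two twist bones.
# Root-relative, spring-driven AnimDynamics then simulates each loop with angular limits; the event graph has no logic.
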AnimBlueprint /BRCosmetics/Characters/Player/Female/Medium/Bodies/F_MED_SkiffGrind/Meshes/F_MED_SkiffGrind_AnimBP
target_skeleton=&/F_MED_SkiffGrind_Skeleton
graphs [
 AnimGraph {nodes [
   n1 = Control Rig(Source=Copy Pose From Mesh() {bUseAttachedParent=true}) {
    alphaBoolBlend {blendOption=Linear}
    controlRigAssetReference {blueprintRigClass=&/F_MED_SkiffGrind_CtrlRig.F_MED_SkiffGrind_CtrlRig_C}
   }
   n2 = "Constraint - dyn_loop_base_l (PP)"!(ComponentPose=Local To Component(LocalPose=@n1), Alpha=1, ConstraintWeights_0=0.858, ConstraintWeights_1:float=0.142) -> (Pose) {
    alphaBoolBlend {blendOption=Linear}
    boneToModify {boneName=dyn_loop_base_l}
    constraintSetup [{targetBone {boneName=upperarm_twist_01_l}}, {targetBone {boneName=dyn_loop_l}}]
    constraintWeights [1,1]
   }
   n3 = "Constraint - dyn_loop_base_r (PP)"!(ComponentPose=@n2, Alpha=1, ConstraintWeights_0=0.969, ConstraintWeights_1:float=0.031) -> (Pose) {
    alphaBoolBlend {blendOption=Linear}
    boneToModify {boneName=dyn_loop_base_r}
    constraintSetup [{targetBone {boneName=upperarm_twist_01_r}}, {targetBone {boneName=upperarm_twist_02_r}}]
    constraintWeights [1,1]
   }
   n4 = Anim Dynamics - Bone: dyn_loop_l(ComponentPose=@n3, Alpha=1, ExternalForce="0,0,0") {
    angularDampingOverride=0.8
    angularSpringConstant=1000
    bAngularSpring=true
    bLinearSpring=true
    bOverrideAngularDamping=true
    bUseGravityOverride=true
    bUsePlanarLimit=false
    gravityScale=0.35
    linearSpringConstant=500
    simulationSpace=RootRelative
    alphaBoolBlend {blendOption=Linear}
    boundBone {boneName=dyn_loop_l}
    componentAppliedLinearAccClamp=<3000 3000 3000>
    componentLinearAccScale=<0.2 0.2 0.2>
    componentLinearVelScale=<0.5 0.5 0.5>
    gravityOverride {z=-650}
    physicsBodyDefinitions [{
      boundBone {boneName=dyn_loop_l}
      boxExtents=<8 8 8>
      constraintSetup {angularLimitsMax=<0 5 3>, angularLimitsMin=<0 -40 -3>, angularTarget=<1 0 0>}
      localJointOffset=<1 0 0>
     }]
   }
   n5 = Anim Dynamics - Bone: dyn_loop_r(ComponentPose=@n4, Alpha=1, ExternalForce="0,0,0") {
    angularDampingOverride=0.8
    angularSpringConstant=1000
    bAngularSpring=true
    bLinearSpring=true
    bOverrideAngularDamping=true
    bUseGravityOverride=true
    bUsePlanarLimit=false
    gravityScale=0.35
    linearSpringConstant=500
    simulationSpace=RootRelative
    alphaBoolBlend {blendOption=Linear}
    boundBone {boneName=dyn_loop_r}
    componentAppliedLinearAccClamp=<3000 3000 3000>
    componentLinearAccScale=<0.2 0.2 0.2>
    componentLinearVelScale=<0.5 0.5 0.5>
    gravityOverride {z=-650}
    physicsBodyDefinitions [{
      boundBone {boneName=dyn_loop_r}
      boxExtents=<8 8 8>
      constraintSetup {angularLimitsMax=<0 5 3>, angularLimitsMin=<0 -40 -3>, angularTarget=<1 0 0>}
      localJointOffset=<1 0 0>
     }]
   }
   Output Pose(Result=Component To Local(ComponentPose=@n5)) {layerGroup=None}
  ]}
 EventGraph {nodes [Event BlueprintUpdateAnimation()]}
]
variables []
