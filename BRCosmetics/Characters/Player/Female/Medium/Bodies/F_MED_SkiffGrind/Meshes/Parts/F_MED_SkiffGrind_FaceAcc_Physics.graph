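# Physics asset for the face accessory mesh of the F_MED_SkiffGrind female character, providing bodies and joints for hair physics.
# A five-segment ponytail and two shorter hair pieces (center and R) are capsule chains with tight twist and per-joint swing limits, each rooted through a base body on the head.
# Head, upper spine and upper-arm capsules add body volumes, linked by angularly free constraints, as are the hair bases to the head.
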PhysicsAsset /BRCosmetics/Characters/Player/Female/Medium/Bodies/F_MED_SkiffGrind/Meshes/Parts/F_MED_SkiffGrind_FaceAcc_Physics
bodies {
 dyn_hair_C [{shapeType=Capsule, length=6, radius=1, shapeName=dyn_hair_C_capsule, center=<0.2994 -0.0636 0>, rotation=<rot 90 -11.9168 0>}]
 dyn_hair_C_basa [{shapeType=Capsule, length=0.505, radius=0.505, shapeName=dyn_hair_C_basa_capsule, rotation=<rot 0 -90 0>}]
 dyn_hair_R [{shapeType=Capsule, length=2, radius=1, shapeName=dyn_hair_R_capsule, center=<0.396 0 0>, rotation=<rot 90 0 0>}]
 dyn_hair_R_basa [{shapeType=Capsule, length=0.505, radius=0.505, shapeName=dyn_hair_R_basa_capsule, rotation=<rot 0 -90 0>}]
 dyn_ponytail_1 [{shapeType=Capsule, length=0, radius=2, shapeName=dyn_ponytail_1_capsule, rotation=<rot 90 0 0>}]
 dyn_ponytail_2 [{shapeType=Capsule, length=8, radius=3, shapeName=dyn_ponytail_2_capsule}]
 dyn_ponytail_3 [{shapeType=Capsule, length=5, radius=5, shapeName=dyn_ponytail_3_capsule}]
 dyn_ponytail_4 [
  {shapeType=Capsule, length=4, radius=4, shapeName=dyn_ponytail_4_capsule, center=<1.3643 0 0>}
  {
   shapeType=Capsule
   length=5
   radius=1.5
   shapeName=dyn_ponytail_4_capsule
   center=<6.5028 -2.5713 4.0921>
   rotation=<rot 75.5418 179.9774 180>
  }
 ]
 dyn_ponytail_5 [{shapeType=Capsule, length=4, radius=4, shapeName=dyn_ponytail_5_capsule, center=<2.702 0 0>, rotation=<rot 90 0 0>}]
 dyn_ponytail_basa [{shapeType=Capsule, length=1, radius=1, shapeName=dyn_ponytail_basa_capsule, rotation=<rot 0 -90 0>}]
 head [{shapeType=Capsule, length=15, radius=7.5, shapeName=head_capsule, rotation=<rot 90 0 0>}]
 spine_05 [{shapeType=Capsule, length=20, radius=8, shapeName=spine_05_capsule, center=<2.1865 0 0>}]
 upperarm_l [{shapeType=Capsule, length=20, radius=7, shapeName=upperarm_l_capsule, center=<13 0 0>, rotation=<rot 90 0 0>}]
 upperarm_r [{shapeType=Capsule, length=20, radius=7, shapeName=upperarm_r_capsule, center=<-13 0 0>, rotation=<rot 90 0 0>}]
}
constraints [
 {bone1Name=dyn_ponytail_5, bone2Name=dyn_ponytail_4, swing1LimitDegrees=15, swing2LimitDegrees=25, twistLimitDegrees=1}
 {bone1Name=dyn_ponytail_4, bone2Name=dyn_ponytail_3, swing1LimitDegrees=30, swing2LimitDegrees=25, twistLimitDegrees=1}
 {bone1Name=dyn_ponytail_3, bone2Name=dyn_ponytail_2, swing1LimitDegrees=50, swing2LimitDegrees=25, twistLimitDegrees=1}
 {bone1Name=dyn_ponytail_2, bone2Name=dyn_ponytail_1, swing1LimitDegrees=50, swing2LimitDegrees=25, twistLimitDegrees=1}
 {bone1Name=dyn_ponytail_1, bone2Name=dyn_ponytail_basa, swing1LimitDegrees=15, swing2LimitDegrees=5, twistLimitDegrees=1}
 {
  bone1Name=dyn_ponytail_basa
  bone2Name=head
  swing1LimitDegrees=45
  swing1Motion=Free
  swing2LimitDegrees=45
  swing2Motion=Free
  twistLimitDegrees=45
  twistMotion=Free
 }
 {bone1Name=dyn_hair_R, bone2Name=dyn_hair_R_basa, swing1LimitDegrees=10, swing2LimitDegrees=3, twistLimitDegrees=1}
 {
  bone1Name=dyn_hair_R_basa
  bone2Name=head
  swing1LimitDegrees=45
  swing1Motion=Free
  swing2LimitDegrees=45
  swing2Motion=Free
  twistLimitDegrees=45
  twistMotion=Free
 }
 {bone1Name=dyn_hair_C, bone2Name=dyn_hair_C_basa, swing1LimitDegrees=5, swing2LimitDegrees=12, twistLimitDegrees=1}
 {
  bone1Name=dyn_hair_C_basa
  bone2Name=head
  swing1LimitDegrees=45
  swing1Motion=Free
  swing2LimitDegrees=45
  swing2Motion=Free
  twistLimitDegrees=45
  twistMotion=Free
 }
 {
  bone1Name=head
  bone2Name=spine_05
  swing1LimitDegrees=45
  swing1Motion=Free
  swing2LimitDegrees=45
  swing2Motion=Free
  twistLimitDegrees=45
  twistMotion=Free
 }
 {
  bone1Name=upperarm_l
  bone2Name=spine_05
  swing1LimitDegrees=45
  swing1Motion=Free
  swing2LimitDegrees=45
  swing2Motion=Free
  twistLimitDegrees=45
  twistMotion=Free
 }
 {
  bone1Name=upperarm_r
  bone2Name=spine_05
  swing1LimitDegrees=45
  swing1Motion=Free
  swing2LimitDegrees=45
  swing2Motion=Free
  twistLimitDegrees=45
  twistMotion=Free
 }
]
