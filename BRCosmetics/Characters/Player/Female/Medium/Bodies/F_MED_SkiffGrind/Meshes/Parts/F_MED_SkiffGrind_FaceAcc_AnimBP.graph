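# Animation Blueprint for the SkiffGrind female body's face accessory part, adding secondary physics motion.
# AnimGraph copies the attached parent's pose through a Control Rig, then runs RigidBody simulation in head-bone space and a spring-driven AnimDynamics on dyn_hair_L.
# Each update sets Gravity Override via a nativized state function using IntenseCello's parameter asset, and sets KneeSliding ON from the emote flag.
# Neither variable is visibly read in the graph.
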AnimBlueprint /BRCosmetics/Characters/Player/Female/Medium/Bodies/F_MED_SkiffGrind/Meshes/Parts/F_MED_SkiffGrind_FaceAcc_AnimBP
target_skeleton=&/F_MED_SkiffGrind_FaceAcc_Skeleton
graphs [
 AnimGraph {nodes [
   n1 = Control Rig(Source=Copy Pose From Mesh() {bUseAttachedParent=true}) {controlRigAssetReference {controlRigAsset=&/F_MED_SkiffGrind_FaceAcc_CtrlRig}}
   n2 = RigidBody!(ComponentPose=Local To Component(LocalPose=@n1), Alpha, OverrideWorldGravity:Vector="0,0,0", ExternalForce) -> (Pose) {
    bClampLinearTranslationLimitToRefPose=true
    bOverrideWorldGravity=true
    cachedBoundsScale=1
    overridePhysicsAsset=&/F_MED_SkiffGrind_FaceAcc_Physics
    simulationSpace=BaseBoneSpace
    alphaBoolBlend {blendOption=Linear}
    baseBoneRef {boneName=head}
    overrideWorldGravity {z=-980}
    simSpaceSettings {
     maxAngularAcceleration=200
     maxAngularVelocity=3
     maxLinearAcceleration=1400
     maxLinearVelocity=250
     worldAlpha=1
     externalLinearDragV=<0.7 0.4 0.4>
     externalLinearVelocity=<2 0.2 0.2>
    }
   }
   n3 = Anim Dynamics - Bone: dyn_hair_L(ComponentPose=@n2, Alpha=1, ExternalForce="0,0,0") {
    angularDampingOverride=0.999
    angularSpringConstant=800
    bAngularSpring=true
    bOverrideAngularDamping=true
    bUseGravityOverride=true
    gravityScale=0.35
    linearSpringConstant=100
    simulationSpace=RootRelative
    alphaBoolBlend {blendOption=Linear}
    boundBone {boneName=dyn_hair_L}
    componentAppliedLinearAccClamp=<3000 3000 3000>
    componentLinearAccScale=<0.2 0.2 0.8>
    componentLinearVelScale=<0.2 0.2 0.8>
    gravityOverride {z=-980}
    physicsBodyDefinitions [{
      boundBone {boneName=dyn_hair_L}
      boxExtents=<20 20 20>
      constraintSetup {angularLimitsMax=<0 10 7>, angularLimitsMin=<0 0 -40>, angularTarget=<1 0 0>}
      localJointOffset=<1 0 0>
     }]
   }
   Output Pose(Result=Component To Local(ComponentPose=@n3)) {layerGroup=None}
  ]}
 EventGraph {nodes [
   Event BlueprintUpdateAnimation()
   > n1 = "Gravity Override State Function (Parameter Asset) (Nativized)"(
    CharacterPartAnimInstance=Self-Reference()
    GameTimeInSeconds=GetGameTimeInSeconds()
    Parameters=/BRCosmetics/Characters/Player/Male/Medium/Bodies/M_MED_IntenseCello/Meshes/M_MED_IntenseCello_Parameters
   )
   > Set Gravity Override(Gravity Override=@n1)
   > Set KneeSliding ON(KneeSliding ON:double="To Float (Boolean)"(InBool=Get bIsPlayingEmote())) -> (Output_Get:double)
  ]}
]
variables [
 Gravity Override {type=Vector}
 KneeSliding ON {type=number}
]
